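# Sound cue for the Hopscotch shotgun's fire sound, built from close, medium and distant layers blended in one mixer.
# Each layer randomly picks a wave, varies it through a modulator, then applies its own attenuation and sound class override, with close using 3P fire and the others distant fire.
# Random nodes are excluded from branch culling.
# Gunfire concurrency sets are assigned, alongside an owner-limited stop-oldest override, and debug is enabled.
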SoundCue /HopscotchWeaponsGameplay/Items/HopscotchShotgun/Audio/Hopscotch_Shotgun_Fire
root=Mixer_0
nodes [
 {id=Attenuation_2, children [Modulator_2], properties {attenuationSettings=&Fort_Weapon_Projectile_Close_Athena_Att}}
 {
  id=Attenuation_3
  children [Modulator_3]
  properties {attenuationSettings=&Fort_Weapon_Projectile_Medium_Mono_Athena_Att}
 }
 {
  id=Attenuation_4
  children [Modulator_4]
  properties {attenuationSettings=&Fort_Weapon_Projectile_Distant_Mono_Athena_Att}
 }
 {id=Mixer_0, children [SoundClass_2,SoundClass_4,SoundClass_3], properties {inputVolume [1.2,0.6,0.5]}}
 {id=Modulator_2, children [Random_0], properties {pitchMax=1.1, pitchMin=1, volumeMax=1.2, volumeMin=1.1}}
 {id=Modulator_3, children [Random_1], properties {volumeMax=2, volumeMin=1.9}}
 {id=Modulator_4, children [Random_2], properties {volumeMax=1.95, volumeMin=1.85}}
 {
  id=Random_0
  children [WavePlayer_1,WavePlayer_2,WavePlayer_3,WavePlayer_4]
  properties {bShouldExcludeFromBranchCulling=true, weights [1,1,1,1]}
 }
 {
  id=Random_1
  children [WavePlayer_9, WavePlayer_10, WavePlayer_11, WavePlayer_12, WavePlayer_13, WavePlayer_14, WavePlayer_15]
  properties {bShouldExcludeFromBranchCulling=true, weights [1,1,1,1,1,1,1]}
 }
 {
  id=Random_2
  children [WavePlayer_26, WavePlayer_27, WavePlayer_28, WavePlayer_29, WavePlayer_30, WavePlayer_31, WavePlayer_32]
  properties {bShouldExcludeFromBranchCulling=true, weights [1,1,1,1,1,1,1]}
 }
 {
  id=SoundClass_2
  children [Attenuation_2]
  properties {soundClassOverride=&Fort_WeaponRanged_Fire_3P_PSM}
 }
 {
  id=SoundClass_3
  children [Attenuation_4]
  properties {soundClassOverride=&Fort_WeaponRanged_Fire_Distant}
 }
 {
  id=SoundClass_4
  children [Attenuation_3]
  properties {soundClassOverride=&Fort_WeaponRanged_Fire_Distant}
 }
 {id=WavePlayer_1, properties {soundWaveAssetPtr=&/Hopscotch_Shotgun_Fire_Close_01}}
 {id=WavePlayer_10, properties {soundWaveAssetPtr=&/Hopscotch_Shotgun_Fire_Medium_02}}
 {id=WavePlayer_11, properties {soundWaveAssetPtr=&/Hopscotch_Shotgun_Fire_Medium_03}}
 {id=WavePlayer_12, properties {soundWaveAssetPtr=&/Hopscotch_Shotgun_Fire_Medium_04}}
 {id=WavePlayer_13, properties {soundWaveAssetPtr=&/Hopscotch_Shotgun_Fire_Medium_05}}
 {id=WavePlayer_14, properties {soundWaveAssetPtr=&/Hopscotch_Shotgun_Fire_Medium_06}}
 {id=WavePlayer_15, properties {soundWaveAssetPtr=&/Hopscotch_Shotgun_Fire_Medium_07}}
 {id=WavePlayer_2, properties {soundWaveAssetPtr=&/Hopscotch_Shotgun_Fire_Close_02}}
 {id=WavePlayer_26, properties {soundWaveAssetPtr=&/Hopscotch_Shotgun_Fire_Distant_01}}
 {id=WavePlayer_27, properties {soundWaveAssetPtr=&/Hopscotch_Shotgun_Fire_Distant_02}}
 {id=WavePlayer_28, properties {soundWaveAssetPtr=&/Hopscotch_Shotgun_Fire_Distant_03}}
 {id=WavePlayer_29, properties {soundWaveAssetPtr=&/Hopscotch_Shotgun_Fire_Distant_04}}
 {id=WavePlayer_3, properties {soundWaveAssetPtr=&/Hopscotch_Shotgun_Fire_Close_03}}
 {id=WavePlayer_30, properties {soundWaveAssetPtr=&/Hopscotch_Shotgun_Fire_Distant_05}}
 {id=WavePlayer_31, properties {soundWaveAssetPtr=&/Hopscotch_Shotgun_Fire_Distant_06}}
 {id=WavePlayer_32, properties {soundWaveAssetPtr=&/Hopscotch_Shotgun_Fire_Distant_07}}
 {id=WavePlayer_4, properties {soundWaveAssetPtr=&/Hopscotch_Shotgun_Fire_Close_04}}
 {id=WavePlayer_9, properties {soundWaveAssetPtr=&/Hopscotch_Shotgun_Fire_Medium_01}}
]
properties {
 bDebug=true
 maxDistance=27500
 soundClassObject=&Fort_WeaponRanged_Fire_3P_PSM
 volumeMultiplier=0.6
 concurrencyOverrides {bLimitToOwner=true, maxCount=3, resolutionRule=StopOldest}
 concurrencySet [&Weapon_Gunfire_Local_Conc, &Weapon_Gunfire_Global_Conc]
}
