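# Sound cue for the first-person firing sound of the Hopscotch shotgun.
# A random node picks evenly from four wave variations, exempt from branch culling, then a modulator adds slight upward pitch variation.
# It uses first-person weapon attenuation, the first-person ranged fire sound class and a shared first-person gunfire concurrency set.
# Its concurrency override values are owner-limited and stop the oldest sound, and the debug flag is on.
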SoundCue /HopscotchWeaponsGameplay/Items/HopscotchShotgun/Audio/Hopscotch_Shotgun_Fire_1P
root=Modulator_0
nodes [
 {id=Modulator_0, children [Random_4], properties {pitchMax=1.02, pitchMin=1}}
 {
  id=Random_4
  children [WavePlayer_19, WavePlayer_20, WavePlayer_21, WavePlayer_22]
  properties {bShouldExcludeFromBranchCulling=true, weights [1,1,1,1]}
 }
 {id=WavePlayer_19, properties {soundWaveAssetPtr=&/Hopscotch_Shotgun_Fire_1P_01}}
 {id=WavePlayer_20, properties {soundWaveAssetPtr=&/Hopscotch_Shotgun_Fire_1P_02}}
 {id=WavePlayer_21, properties {soundWaveAssetPtr=&/Hopscotch_Shotgun_Fire_1P_03}}
 {id=WavePlayer_22, properties {soundWaveAssetPtr=&/Hopscotch_Shotgun_Fire_1P_04}}
]
properties {
 attenuationSettings=&Fort_Weapon_Projectile_FirstPerson_Athena_Att
 bDebug=true
 maxDistance=2097152
 soundClassObject=&Fort_WeaponRanged_Fire_1P_PSM
 volumeMultiplier=1
 concurrencyOverrides {bLimitToOwner=true, maxCount=21, resolutionRule=StopOldest}
 concurrencySet [&Weapon_Gunfire_FirstPerson_Conc]
}
