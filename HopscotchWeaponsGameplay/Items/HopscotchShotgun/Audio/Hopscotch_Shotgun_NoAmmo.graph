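# Sound cue for the Hopscotch Shotgun's no-ammo sound.
# A Random node picks among four equally weighted wave variations and is excluded from branch culling.
# A Modulator at the root applies default pitch and volume variation to the pick.
# The cue plays in the weapon foley sound class and borrows the shared weapon reload attenuation.
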SoundCue /HopscotchWeaponsGameplay/Items/HopscotchShotgun/Audio/Hopscotch_Shotgun_NoAmmo
root=Modulator_0
nodes [
 {id=Modulator_0, children [Random_0]}
 {
  id=Random_0
  children [WavePlayer_0,WavePlayer_3,WavePlayer_4,WavePlayer_6]
  properties {bShouldExcludeFromBranchCulling=true, weights [1,1,1,1]}
 }
 {id=WavePlayer_0, properties {soundWaveAssetPtr=&/Hopscotch_Shotgun_NoAmmo_01}}
 {id=WavePlayer_3, properties {soundWaveAssetPtr=&/Hopscotch_Shotgun_NoAmmo_02}}
 {id=WavePlayer_4, properties {soundWaveAssetPtr=&/Hopscotch_Shotgun_NoAmmo_03}}
 {id=WavePlayer_6, properties {soundWaveAssetPtr=&/Hopscotch_Shotgun_NoAmmo_04}}
]
properties {
 attenuationSettings=&Fort_Weapon_Reload_Att
 maxDistance=2250
 soundClassObject=&Fort_WeaponFoley
 volumeMultiplier=0.902
}
