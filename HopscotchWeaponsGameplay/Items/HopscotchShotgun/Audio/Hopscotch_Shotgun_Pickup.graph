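# Sound cue for picking up the Hopscotch shotgun, classed as loot-collect audio with weapon pickup attenuation.
# A random node picks evenly among five pickup wave variations, is excluded from branch culling, and feeds a modulator at the root.
# Overridden concurrency allows one instance at a time and blocks new plays while it is active.
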SoundCue /HopscotchWeaponsGameplay/Items/HopscotchShotgun/Audio/Hopscotch_Shotgun_Pickup
root=Modulator_0
nodes [
 {id=Modulator_0, children [Random_1]}
 {
  id=Random_1
  children [WavePlayer_2,WavePlayer_4,WavePlayer_5,WavePlayer_6,WavePlayer_7]
  properties {bShouldExcludeFromBranchCulling=true, weights [1,1,1,1,1]}
 }
 {id=WavePlayer_2, properties {soundWaveAssetPtr=&/Hopscotch_Shotgun_PickUp_01}}
 {id=WavePlayer_4, properties {soundWaveAssetPtr=&/Hopscotch_Shotgun_PickUp_02}}
 {id=WavePlayer_5, properties {soundWaveAssetPtr=&/Hopscotch_Shotgun_PickUp_03}}
 {id=WavePlayer_6, properties {soundWaveAssetPtr=&/Hopscotch_Shotgun_PickUp_04}}
 {id=WavePlayer_7, properties {soundWaveAssetPtr=&/Hopscotch_Shotgun_PickUp_05}}
]
properties {
 attenuationSettings=&Fort_Weapon_Pickup_Att
 bOverrideConcurrency=true
 maxDistance=3500
 soundClassObject=&Fort_Loot_Collect
 volumeMultiplier=0.6
 concurrencyOverrides {maxCount=1, resolutionRule=PreventNew}
}
