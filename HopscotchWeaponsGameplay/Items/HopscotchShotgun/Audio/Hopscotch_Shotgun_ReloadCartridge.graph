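# Sound cue for inserting a cartridge while reloading the Hopscotch shotgun.
# A Random node picks among eight equally weighted cartridge-insert wave variations and feeds a Modulator left at default settings, which is the cue's output.
# The Random node is excluded from branch culling, and the cue uses the shared Fortnite weapon-reload attenuation and weapon foley sound class.
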SoundCue /HopscotchWeaponsGameplay/Items/HopscotchShotgun/Audio/Hopscotch_Shotgun_ReloadCartridge
root=Modulator_0
nodes [
 {id=Modulator_0, children [Random_1]}
 {
  id=Random_1
  children [WavePlayer_3, WavePlayer_4, WavePlayer_5, WavePlayer_6, WavePlayer_7, WavePlayer_8, WavePlayer_9, WavePlayer_10]
  properties {bShouldExcludeFromBranchCulling=true, weights [1,1,1,1,1,1,1,1]}
 }
 {
  id=WavePlayer_10
  properties {soundWaveAssetPtr=&/Hopscotch_Shotgun_Insert_Cartridge_08}
 }
 {id=WavePlayer_3, properties {soundWaveAssetPtr=&/Hopscotch_Shotgun_Insert_Cartridge_01}}
 {id=WavePlayer_4, properties {soundWaveAssetPtr=&/Hopscotch_Shotgun_Insert_Cartridge_02}}
 {id=WavePlayer_5, properties {soundWaveAssetPtr=&/Hopscotch_Shotgun_Insert_Cartridge_03}}
 {id=WavePlayer_6, properties {soundWaveAssetPtr=&/Hopscotch_Shotgun_Insert_Cartridge_04}}
 {id=WavePlayer_7, properties {soundWaveAssetPtr=&/Hopscotch_Shotgun_Insert_Cartridge_05}}
 {id=WavePlayer_8, properties {soundWaveAssetPtr=&/Hopscotch_Shotgun_Insert_Cartridge_06}}
 {id=WavePlayer_9, properties {soundWaveAssetPtr=&/Hopscotch_Shotgun_Insert_Cartridge_07}}
]
properties {
 attenuationSettings=&Fort_Weapon_Reload_Att
 maxDistance=2250
 soundClassObject=&Fort_WeaponFoley
 volumeMultiplier=1.002
}
