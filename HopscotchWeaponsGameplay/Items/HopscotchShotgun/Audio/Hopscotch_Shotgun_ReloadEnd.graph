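# Sound cue for the end-of-reload sound of the Hopscotch Shotgun.
# A Random node picks among four equally weighted reload-end waves and feeds a Modulator at its default settings for pitch and volume variation.
# The cue lowers the volume, is classed as weapon foley and uses the shared Fortnite weapon reload attenuation.
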SoundCue /HopscotchWeaponsGameplay/Items/HopscotchShotgun/Audio/Hopscotch_Shotgun_ReloadEnd
root=Modulator_0
nodes [
 {id=Modulator_0, children [Random_1]}
 {
  id=Random_1
  children [WavePlayer_1,WavePlayer_2,WavePlayer_5,WavePlayer_6]
  properties {bShouldExcludeFromBranchCulling=true, weights [1,1,1,1]}
 }
 {id=WavePlayer_1, properties {soundWaveAssetPtr=&/Hopscotch_Shotgun_Reload_End_04}}
 {id=WavePlayer_2, properties {soundWaveAssetPtr=&/Hopscotch_Shotgun_Reload_End_01}}
 {id=WavePlayer_5, properties {soundWaveAssetPtr=&/Hopscotch_Shotgun_Reload_End_02}}
 {id=WavePlayer_6, properties {soundWaveAssetPtr=&/Hopscotch_Shotgun_Reload_End_03}}
]
properties {
 attenuationSettings=&Fort_Weapon_Reload_Att
 maxDistance=2250
 soundClassObject=&Fort_WeaponFoley
 volumeMultiplier=0.376
}
